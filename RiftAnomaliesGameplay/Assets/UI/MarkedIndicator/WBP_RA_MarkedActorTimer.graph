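# Rift Anomalies widget that shows a time-remaining label for a marked actor indicator.
# A single CommonTextBlock in an overlay starts blank on Construct, clearing its placeholder text.
# The UpdateTimer event takes the remaining seconds and shows them through TimeToText, or blanks the label at exactly zero.
# The widget runs no countdown itself: its timer handle, remaining-time and frequency variables are unused in the graph.
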WidgetBlueprint /RiftAnomaliesGameplay/Assets/UI/MarkedIndicator/WBP_RA_MarkedActorTimer
graphs [EventGraph {nodes [
   n1 = UpdateTimer!() -> (OutputDelegate, then, InTimeRemaining:float=~) {is=title K2Node_CustomEvent}
   > "SetText (Text)"(
    self=Get Text_Timer() -> (Text_Timer:Common Text Object Reference)
    InText=SelectText(
     B=TimeToText!(Time:float=@n1.InTimeRemaining) -> (ReturnValue:Text)
     bPickA="Nearly Equal (Float)"(A=@n1.InTimeRemaining, B=0, ErrorTolerance=0)
    )
   )
   Event Construct()
   > "SetText (Text)"(self=Get Text_Timer())
  ]}]
variables [
 CountdownTimerHandle {type=TimerHandle}
 TimeRemaining {type=number}
 TimerUpdateFrequency {type=number, default=1}
]
widgets {
 info {inheritedWidgetCount=0, parentClass=&UserWidget, rootWidgetClass=&Overlay, widgetCount=2}
 widgets [
  Overlay_16 {widgetClassPath=&Overlay, widget=Overlay_16}
  Text_Timer {
   widgetClassPath=&CommonTextBlock
   bIsVariable=true
   parent=Overlay_16
   widget=Text_Timer
   properties {
    text=3:00
    font {
     bForceMonospaced=false
     bMaterialIsStencil=false
     fontMaterial=null
     fontObject=&Burbank
     letterSpacing=0
     monospacedWidth=1
     size=14
     skewAmount=0.2
     typefaceFontName=BurbankRegular-Black
     outlineSettings {
      bApplyOutlineToDropShadows=false
      bMiteredCorners=false
      bSeparateFillAlpha=false
      outlineMaterial=null
      outlineSize=1
      outlineColor=<rgba 0 0 0 0.6>
     }
    }
    shadowColorAndOpacity=<rgba 0 0 0 1>
    shadowOffset=<3 3>
   }
   slot {class=OverlaySlot, properties {horizontalAlignment=HAlign_Fill, verticalAlignment=VAlign_Fill}}
  }
 ]
}
